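# FrankenCheeks_AnimBP_Child is an Animation Blueprint for the FrankenCheeks biped skeleton in CosmeticCompanions, named as a child variant.
# Its EventGraph is empty and no animation graph or parent class appears, so it shows no logic of its own.
# It declares two numeric variables, GoalRotationDelta and GoalRotationYawLastUpdate, whose names suggest goal yaw change tracking, though nothing here reads or writes them.
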AnimBlueprint /CosmeticCompanions/Assets/Biped/FrankenCheeks/Blueprint/FrankenCheeks_AnimBP_Child
target_skeleton=/CosmeticCompanions/Assets/Biped/FrankenCheeks/Meshes/Biped/FrankenCheeks_Biped_Skeleton
graphs [EventGraph {links [], nodes []}]
variables [
 GoalRotationDelta {type=number}
 GoalRotationYawLastUpdate {type=number}
]
